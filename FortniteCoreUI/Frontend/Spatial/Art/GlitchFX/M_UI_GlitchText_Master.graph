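# Translucent UI master material that applies an animated glitch to a texture parameter, intended for text by its name.
# Two panning soft-noise masks combine with two scrolling scanline layers, whose tiling oscillates on layered sine waves.
# These drive noise-gated horizontal UV distortion of the texture.
# The sampled alpha is multiplied by the same noise-scanline mask, which can be inverted or filled white, and an overall Alpha.
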Material /FortniteCoreUI/Frontend/Spatial/Art/GlitchFX/M_UI_GlitchText_Master
graph {
 properties {
  bUsedWithStaticMesh=false
  blendMode=BLEND_Translucent
  materialDomain=MD_UI
  pixelDepthOffsetMode=PDOM_Legacy
  usedShadingModels=Default Lit
 }
 nodes [
  Constant() {r=1}
  n1 = TextureCoordinate()
  n2 = ScalarParameter() {defaultValue=0.05, group=EffectAmount, parameterName=DistortionDistance}
  n3 = TextureCoordinate()
  n4 = AppendVector(
   A=ScalarParameter() {defaultValue=0.5, group=Mask, parameterName=SoftNoiseMask_1_UScale}
   B=ScalarParameter() {defaultValue=0.5, group=Mask, parameterName=SoftNoiseMask_1_VScale}
  )
  n5 = ScalarParameter() {defaultValue=-0.4, group=Mask, parameterName=SoftNoiseMask_1_USpeed}
  n6 = ScalarParameter() {defaultValue=0.45, group=Mask, parameterName=SoftNoiseMask_1_VSpeed}
  n7 = Multiply(A=Time(), B=ScalarParameter() {defaultValue=1, parameterName=GlobalTime})
  n8 = TextureSample(UVs=Panner(Coordinate=Multiply(A=@n3, B=@n4), Time=@n7, Speed=AppendVector(A=@n5, B=@n6))) {samplerType=SAMPLERTYPE_Grayscale, texture=&LowResBlurredNoise}
  n9 = TextureCoordinate()
  n10 = AppendVector(
   A=ScalarParameter() {defaultValue=0.5, group=Mask, parameterName=SoftNoiseMask_2_UScale}
   B=ScalarParameter() {defaultValue=0.5, group=Mask, parameterName=SoftNoiseMask_2_VScale}
  )
  n11 = ScalarParameter() {defaultValue=0.88, group=Mask, parameterName=SoftNoiseMask_2_USpeed}
  n12 = ScalarParameter() {defaultValue=-0.11, group=Mask, parameterName=SoftNoiseMask_2_VSpeed}
  n13 = TextureSample(UVs=Panner(Coordinate=Multiply(A=@n9, B=@n10), Time=@n7, Speed=AppendVector(A=@n11, B=@n12))) {samplerType=SAMPLERTYPE_Grayscale, texture=&LowResBlurredNoise}
  n14 = Power(Base=Multiply(A=DotProduct(A=@n8.G, B=@n13.G)) {constB=3}) {constExponent=3}
  n15 = Saturate(None=Multiply(A=Saturate(None=@n14)) {constB=16})
  n16 = Multiply(A=LinearInterpolate(A=Multiply(A=@n2) {constB=-1}, B=@n2, Alpha=@n15)) {constB=4}
  n17 = ScalarParameter() {defaultValue=8, group=Scanlines, parameterName=Tex1_TileAmount_Low}
  n18 = ScalarParameter() {defaultValue=12, group=Scanlines, parameterName=Tex1_TileAmount_High}
  n19 = Multiply(A=Time(), B=ScalarParameter() {defaultValue=1, group=Time, parameterName=GlobalTime})
  n20 = Multiply(A=@n19, B=Constant() {r=1})
  n21 = Add(A=ComponentMask(None=Sine(None=@n20)) {r=true})
  n22 = Sine(None=Multiply(A=@n19, B=Constant() {r=3}))
  n23 = LinearInterpolate(Alpha=Add(A=ComponentMask(None=@n22) {r=true}))
  n24 = Sine(None=Multiply(A=@n19, B=Constant() {r=0.5}))
  n25 = LinearInterpolate(Alpha=Add(A=ComponentMask(None=@n24) {r=true}))
  n26 = Saturate(None=DotProduct(A=LinearInterpolate(Alpha=@n21), B=DotProduct(A=@n23, B=@n25)))
  n27 = Multiply(A=TextureCoordinate(), B=AppendVector(A=Constant(), B=LinearInterpolate(A=@n17, B=@n18, Alpha=@n26)))
  n28 = AppendVector(A=Constant(), B=ScalarParameter() {defaultValue=-1, group=Scanlines, parameterName=Tex1_VSpeed})
  n29 = Multiply(A=TextureSample(UVs=Panner(Coordinate=@n27, Time=@n7, Speed=@n28)) {samplerType=SAMPLERTYPE_Grayscale, texture=&LowResBlurredNoise}.G) {constB=3}
  n30 = ScalarParameter() {defaultValue=1, group=Scanlines, parameterName=Tex2_TileAmount_Low}
  n31 = ScalarParameter() {defaultValue=3, group=Scanlines, parameterName=Tex2_TileAmount_High}
  n32 = Multiply(A=@n19, B=Constant() {r=2})
  n33 = Add(A=ComponentMask(None=Sine(None=@n32)) {r=true})
  n34 = Sine(None=Multiply(A=@n19, B=Constant() {r=6}))
  n35 = LinearInterpolate(Alpha=Add(A=ComponentMask(None=@n34) {r=true}))
  n36 = Sine(None=Multiply(A=@n19, B=Constant() {r=1}))
  n37 = LinearInterpolate(Alpha=Add(A=ComponentMask(None=@n36) {r=true}))
  n38 = Saturate(None=DotProduct(A=LinearInterpolate(Alpha=@n33), B=DotProduct(A=@n35, B=@n37)))
  n39 = Multiply(A=TextureCoordinate(), B=AppendVector(A=Constant(), B=LinearInterpolate(A=@n30, B=@n31, Alpha=@n38)))
  n40 = AppendVector(A=Constant(), B=ScalarParameter() {defaultValue=0.25, group=Scanlines, parameterName=Tex2_VSpeed})
  n41 = Multiply(A=TextureSample(UVs=Panner(Coordinate=@n39, Time=@n7, Speed=@n40)) {samplerType=SAMPLERTYPE_Grayscale, texture=&LowResBlurredNoise}.G) {constB=3}
  n42 = DotProduct(A=Saturate(None=Power(Base=@n29)), B=Saturate(None=Power(Base=@n41)))
  n43 = Saturate(None=Power(Base=@n42) {constExponent=4})
  n44 = Add(A=@n1, B=AppendVector(A=Multiply(A=@n16, B=@n43), B=Constant()))
  n45 = OneMinus(None=@n8.G)
  n46 = OneMinus(None=@n13.G)
  n47 = Saturate(None=LinearInterpolate(
    A=ScalarParameter() {defaultValue=-4, group=EffectAmount, parameterName=DistortionAmount}
    B=Constant() {r=2}
    Alpha=DotProduct(A=@n45, B=@n46)
   ))
  n48 = DotProduct(A=@n15, B=@n43)
  n49 = LinearInterpolate(A=@n48, B=OneMinus(None=@n48), Alpha=ScalarParameter() {parameterName=FlipMask})
  n50 = Constant() {r=1}
  n51 = ScalarParameter() {parameterName=FIllWhite}
  DotProduct(
   A=TextureSampleParameter2D(UVs=LinearInterpolate(A=@n1, B=@n44, Alpha=@n47)) {parameterName=Texture, texture=/Game/Athena/HUD/Art/BacchusHudIcons/T_UI_StandardClipMode_128x}.A
   B=Multiply(A=LinearInterpolate(A=@n49, B=@n50, Alpha=@n51), B=ScalarParameter() {defaultValue=1, parameterName=Alpha})
  )
  Multiply(A=@n42) {constB=4}
 ]
}
